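# Niagara system for the secondary trail of the DebitScent emote, pairing layered ribbons with drifting icon sprites.
# FN_AnimTrail builds bone-driven ribbons with white, cyan and purple layers, its lifetime scaled by User.SpawnRate.
# Icons spawns three visibility-tagged groups from the skeletal mesh, spinning and drifting through curl noise and drag.
# Both taper over emitter loop age and adjust to time of day; older MainTrail and GlowTrail emitters are disabled.
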NiagaraSystem /BRCosmetics/Animation/Game/MainPlayer/Emotes/DebitScent/FX/NS_Emote_DebitScent_SecondTrail
dependencies {
 usedDataInterfaces [
  &NiagaraDataInterfaceCurve
  &NiagaraDataInterfaceSkeletalMesh
  &NiagaraDataInterfaceSocketReader
  &NiagaraDataInterfaceArrayFloat
 ]
 usedDynamicInputs [
  &MultiplyLinearColor
  &MakeVector4_FromLinearColor
  &SelectFloatFromArray
 ]
 usedModules [
  &self:SystemSpawnScript.SystemScriptSource.SystemScriptGraph.NiagaraNodeAssignment_0.SetVariables_#1
  &self:SystemSpawnScript.SystemScriptSource.SystemScriptGraph.NiagaraNodeAssignment_1.SetVariables_#2
  &SystemState
  &self:SystemSpawnScript.SystemScriptSource.SystemScriptGraph.NiagaraNodeAssignment_3.SetVariables_#3
 ]
}
emitters {
 FN_AnimTrail {
  emitterSpawnScript [&FN_RibbonIDOffsetUtility()]
  emitterUpdateScript [
   &EmitterState()
   &FN_SpawnAnimTrail()
  ]
  particleSpawnScript [
   &FN_InitializeAnimTrailBoneIndecies()
   &FN_InitializeAnimTrailBoneIndecies(AnimTrail Selection=1, First Bone or Socket Index=2, Second Bone or Socket Index=3) {enabled=false, module=FN_InitializeAnimTrailBoneIndecies001}
   &FN_InitializeAnimTrail(AnimTrailWidth=2.75, Lifetime=&Multiply_Float(A=0.35, B=$User.SpawnRate))
   &AddVelocity(Velocity=<0 0 500>) {enabled=false}
  ]
  particleUpdateScript [
   &SolveForcesAndVelocity(Clamp Velocity=false, Limit Acceleration=false)
   &FN_UpdateAnimTrail() {module=FN_UpdateAnimTrail001}
   &ScaleColor(
    Scale Alpha=&FloatFromCurve(FloatCurve=DataInterface() {propertyValues {Curve {keys [
         <at 0 0 RCIM_Cubic>
         {arriveTangent=-0.1027, interpMode=RCIM_Cubic, leaveTangent=-0.1027, tangentMode=RCTM_User, time=0.1, value=1}
         {arriveTangent=-1.1111, interpMode=RCIM_Cubic, leaveTangent=-1.1111, time=0.3762, value=0.3007}
         <at 1 0 RCIM_Cubic>
        ]}}})
    Scale RGB=&VectorFromFloat(Value=&FNTimeOfDayScalar(Night Scale=0.7, Morning Scale=1, Evening Scale=1))
    ScaleA=true
    ScaleRGB=true
   )
   &ScaleRibbonWidth(Ribbon Width Scale=&Multiply_Float(
     A=&FloatFromCurve(FloatCurve=DataInterface() {propertyValues {Curve {keys [
          {arriveTangent=5.2183, interpMode=RCIM_Cubic, leaveTangent=5.2183, tangentMode=RCTM_User, time=0, value=0}
          <at 0.2 1 RCIM_Cubic>
          <at 1 0 RCIM_Cubic>
         ]}}})
     B=&FloatFromCurve(
      FloatCurve=DataInterface() {propertyValues {Curve {keys [<at 0 1>, {arriveTangent=-1, interpMode=RCIM_Cubic, leaveTangent=-1, time=0.85, value=1}, <at 1 0 RCIM_Cubic>]}}}
      CurveIndex=$Emitter.NormalizedLoopAge
     )
    ))
   /SetVariables(
    Particles.1 Color:LinearColor=&MakeLinearColorFromVectorAndFloat(
     "Vector (RGB)"=<1 1 1>
     "Float (Alpha)"=&FloatFromCurve(
      FloatCurve=DataInterface() {propertyValues {Curve {keys [
          <at 0 0 RCIM_Cubic>
          {arriveTangent=-0.1027, interpMode=RCIM_Cubic, leaveTangent=-0.1027, tangentMode=RCTM_User, time=0.1, value=1}
          {arriveTangent=-1.1111, interpMode=RCIM_Cubic, leaveTangent=-1.1111, time=0.3762, value=0.3007}
          <at 1 0 RCIM_Cubic>
         ]}}}
      Scale Curve=0.475
     )
    )
    Particles.1 Width:NiagaraFloat=&Multiply_Float(A=$Particles.RibbonWidth, B=0.65)
   ) {isSetParameters=true}
   /SetVariables(
    Particles.2 Color:LinearColor=&MakeLinearColorFromVectorAndFloat(
     "Vector (RGB)"=&MakeVectorFromLinearColorRGB(Color=&Multiply_LinearColorByFloat(
       Linear Color=<rgba 0 8.3356 12 5>
       Float=&FNTimeOfDayScalar(Night Scale=0.7, Morning Scale=1, Evening Scale=1)
      ))
     "Float (Alpha)"=&FloatFromCurve(
      FloatCurve=DataInterface() {propertyValues {Curve {keys [
          <at 0 0 RCIM_Cubic>
          {arriveTangent=-0.1027, interpMode=RCIM_Cubic, leaveTangent=-0.1027, tangentMode=RCTM_User, time=0.1, value=1}
          {arriveTangent=-1.6667, interpMode=RCIM_Cubic, leaveTangent=-1.6667, time=0.3762, value=0.3007}
          <at 0.7 0 RCIM_Cubic>
         ]}}}
      Scale Curve=0.6
     )
    )
    Particles.2 Width:NiagaraFloat=&Multiply_Float(A=$Particles.RibbonWidth, B=0.45)
    Particles.2DP:Vector4f=&MakeVector4(X=0, Y=0, Z=2, W=&Multiply_Float(A=$Engine.DeltaTime, B=-5))
   ) {isSetParameters=true}
   /SetVariables(
    Particles.3 Color:LinearColor=&MakeLinearColorFromVectorAndFloat(
     "Vector (RGB)"=&MakeVectorFromLinearColorRGB(Color=&Multiply_LinearColorByFloat(
       Linear Color=<rgba 0.1648 0.0549 0.4219 1>
       Float=&FNTimeOfDayScalar(Night Scale=0.7, Morning Scale=1, Evening Scale=1)
      ))
     "Float (Alpha)"=&FloatFromCurve(FloatCurve=DataInterface() {propertyValues {Curve {keys [
          <at 0 0 RCIM_Cubic>
          {arriveTangent=-0.1027, interpMode=RCIM_Cubic, leaveTangent=-0.1027, tangentMode=RCTM_User, time=0.1, value=1}
          {arriveTangent=-1.5529, interpMode=RCIM_Cubic, leaveTangent=-1.5529, tangentMode=RCTM_User, time=1, value=0}
         ]}}})
    )
    Particles.3DP:Vector4f=&MakeVector4(
     X=&FloatFromCurve(FloatCurve=DataInterface() {propertyValues {Curve {keys [<at 0 1 RCIM_Cubic>, {arriveTangent=-3.1224, interpMode=RCIM_Cubic, leaveTangent=-3.1224, tangentMode=RCTM_User, time=1, value=0}]}}})
     Y=1
     Z=1
     W=0.425
    )
    Particles.3Width:NiagaraFloat=&Multiply_Float(A=$Particles.RibbonWidth, B=0.4)
   ) {isSetParameters=true}
  ]
  rendererClasses [&NiagaraRibbonRendererProperties]
  renderers [
   {rendererClass=&NiagaraRibbonRendererProperties}
   {rendererClass=&NiagaraRibbonRendererProperties, rendererIndex=1}
   {rendererClass=&NiagaraRibbonRendererProperties, rendererIndex=2}
   {rendererClass=&NiagaraRibbonRendererProperties, rendererIndex=3}
   {rendererClass=&NiagaraRibbonRendererProperties, rendererIndex=4}
   {rendererClass=&NiagaraRibbonRendererProperties, rendererIndex=5}
  ]
 }
 GlowTrail {
  bEnabled=false
  emitterUpdateScript [
   &EmitterState()
   &SpawnRate(SpawnRate=&Multiply_Float(A=$User.SpawnRate, B=120), Use Spawn Probability=false)
  ]
  particleSpawnScript [
   &InitializeParticle(
    Color=<rgba 0.1648 0.0549 0.4219 1>
    Lifetime=0.3
    Ribbon Width=20
    Ribbon Width Mode=::NewEnumerator1
    UsePositionOffset=false
    Write Material Random=true
   )
   &SkeletalMeshLocation(Bone Sampling Mode=::NewEnumerator4, MaskSpawnGroup=false, Skeletal Mesh=$System.Skeletal Mesh Reader)
  ]
  particleUpdateScript [
   &ParticleState()
   &ScaleColor(
    Scale Alpha=&FloatFromCurve(FloatCurve=DataInterface() {propertyValues {Curve {keys [
         <at 0 0 RCIM_Cubic>
         {arriveTangent=-0.1027, interpMode=RCIM_Cubic, leaveTangent=-0.1027, tangentMode=RCTM_User, time=0.1, value=1}
         {arriveTangent=-1.5529, interpMode=RCIM_Cubic, leaveTangent=-1.5529, tangentMode=RCTM_User, time=1, value=0}
        ]}}})
    ScaleA=true
    ScaleRGB=true
   )
   &ScaleRibbonWidth(Ribbon Width Scale=&FloatFromCurve(FloatCurve=DataInterface() {propertyValues {Curve {keys [
         {arriveTangent=6.1048, interpMode=RCIM_Cubic, leaveTangent=6.1048, tangentMode=RCTM_User, time=0, value=0}
         <at 0.2 1 RCIM_Cubic>
         <at 1 0 RCIM_Cubic>
        ]}}}))
   &DynamicMaterialParameters(
    Index 0 Param 1=&FloatFromCurve(FloatCurve=DataInterface() {propertyValues {Curve {keys [<at 0 1 RCIM_Cubic>, {arriveTangent=-3.1224, interpMode=RCIM_Cubic, leaveTangent=-3.1224, tangentMode=RCTM_User, time=1, value=0}]}}})
    Index 0 Param 2=1
    Index 0 Param 3=1
    Index 0 Param 4=0.425
    Module.Index0Float0Write=true
    Module.Index0Float1Write=true
    Module.Index0Float2Write=true
    Module.Index0Float3Write=true
   )
  ]
  rendererClasses [&NiagaraRibbonRendererProperties]
  renderers [{rendererClass=&NiagaraRibbonRendererProperties}]
 }
 Icons {
  emitterUpdateScript [
   &EmitterState()
   &SpawnRate(Spawn Probability=0.85, SpawnRate=&Multiply_Float(A=$User.SpawnRate, B=16), Use Spawn Probability=true)
   &SpawnRate(
    Spawn Probability=0.85
    SpawnGroup=1
    SpawnRate=&Multiply_Float(A=$User.SpawnRate, B=10)
    Use Spawn Probability=true
   ) {module=SpawnRate001}
   &SpawnRate(
    Spawn Probability=0.85
    SpawnGroup=2
    SpawnRate=&Multiply_Float(A=$User.SpawnRate, B=28)
    Use Spawn Probability=false
   ) {module=SpawnRate002}
  ]
  particleSpawnScript [
   /SetVariables(Particles.VisibilityTag=&SpawnGroup()) {isSetParameters=true}
   &InitializeParticle(
    Color=<rgba 0.6 0.6 0.6 1>
    Lifetime Max=1
    Lifetime Min=0.225
    Lifetime Mode=::NewEnumerator1
    Sprite Rotation Mode=::NewEnumerator2
    Sprite Size Mode=::NewEnumerator3
    Uniform Sprite Size=&RandomRangeFloat(Minimum=15, Maximum=24)
    UsePositionOffset=false
    Write Material Random=true
   )
   &ShapeLocation(
    MaskSpawnGroup=false
    Override Seed=false
    Sphere Radius=&MergeHelper_FloatSwitch(0=4, 1=6, 2=8, 3=0, 4=0, Index=&SpawnGroup(), Num=::NewEnumerator3)
    UseOwnerScale=true
   )
   &SkeletalMeshLocation(
    Bone Sampling Mode=::NewEnumerator4
    "Inherit Velocity (Scale)"=0.55
    Inherited Velocity Speed Limit=200
    MaskSpawnGroup=false
    Sampled Position Offset:NiagaraPosition=$Particles.ShapeLocation.ShapePosition
    Skeletal Mesh=$System.Skeletal Mesh Reader
    Velocity Sampling=::NewEnumerator0
    VelocitySpeedLimit=true
   )
   &AddVelocity(
    Cone Axis=&RotateVectorByQuaternion(Vector To Rotate=<0 1 0>, Quaternion=$Particles.SkeletalMeshLocation.SampledRotation)
    Module.UseConeFalloff=false
    Module.UseVelDistribution=false
    Override Seed=false
    Velocity Mode=::NewEnumerator2
    Velocity Speed=&RandomRangeFloat(Minimum=25, Maximum=40)
   )
   /SetVariables(Particles.SubImageIndex=&RandomRangeFloat(Minimum=0, Maximum=2)) {isSetParameters=true}
   /SetVariables(Particles.SubImageIndex2:NiagaraFloat=&RandomRangeFloat(Minimum=3, Maximum=4)) {isSetParameters=true}
   /SetVariables(Particles.SubImageIndex3:NiagaraFloat=&RandomRangeFloat(Minimum=2, Maximum=3)) {isSetParameters=true}
  ]
  particleUpdateScript [
   &ParticleState()
   &SpriteRotationRate(Rotation Rate=/Niagara/DynamicInputs/Helpers/DistributionCurveFloat(
     Minimum:NiagaraFloat=-250
     Maximum:NiagaraFloat=250
     Distribution Curve:NiagaraDataInterfaceCurve=DataInterface() {propertyValues {Curve {
        defaultValue=0
        keys [
         {arriveTangent=-9.8387, interpMode=RCIM_Cubic, leaveTangent=-9.8387, tangentMode=RCTM_User, time=0, value=1}
         {arriveTangent=-0.0373, interpMode=RCIM_Cubic, leaveTangent=-0.0373, tangentMode=RCTM_User, time=0.25, value=0}
         {arriveTangent=0.1077, interpMode=RCIM_Cubic, leaveTangent=0.1077, tangentMode=RCTM_User, time=0.75, value=0}
         {arriveTangent=9.3174, interpMode=RCIM_Cubic, leaveTangent=9.3174, tangentMode=RCTM_User, time=1, value=1}
        ]
       }}}
     Curve Parameter Type:NiagaraInt32=0
     "Curve Parameter Type:":NiagaraInt32=0
     Curve Parameter:NiagaraFloat=?
     Evaluation Type:ENiagaraRandomnessEvaluation=::NewEnumerator0
     Recalculate Random Each Loop:NiagaraBool=true
     Randomness Mode:ENiagaraRandomnessMode=::NewEnumerator0
     Random Seed:NiagaraInt32=0
     Fixed Random Seed:NiagaraBool=false
    ))
   &ScaleSpriteSize(
    Scale Sprite Size Mode=::NewEnumerator0
    Uniform Scale Factor=&Multiply_Float(
     A=&FloatFromCurve(FloatCurve=DataInterface() {propertyValues {Curve {
         defaultValue=0
         keys [
          {arriveTangent=9.0725, interpMode=RCIM_Cubic, leaveTangent=9.0725, tangentMode=RCTM_User, time=0, value=0}
          {arriveTangent=-0.0383, interpMode=RCIM_Cubic, leaveTangent=-0.0383, tangentMode=RCTM_User, time=0.1408, value=0.9847}
          {arriveTangent=-1.8894, interpMode=RCIM_Cubic, leaveTangent=-1.8894, tangentMode=RCTM_User, time=1, value=0}
         ]
        }}})
     B=&FloatFromCurve(
      FloatCurve=DataInterface() {propertyValues {Curve {
         defaultValue=0
         keys [
          {tangentMode=RCTM_User, time=0, value=1}
          {arriveTangent=-0.0405, interpMode=RCIM_Cubic, leaveTangent=-0.0405, tangentMode=RCTM_User, time=0.7, value=1}
          {arriveTangent=-13.6977, interpMode=RCIM_Cubic, leaveTangent=-13.6977, tangentMode=RCTM_User, time=0.85, value=0}
         ]
        }}}
      CurveIndex=$Emitter.NormalizedLoopAge
     )
    )
   )
   /SetVariables(
    Particles.SmallIconColor:LinearColor=&FNTimeOfDayColor(Night Color=<rgba 0.5 0.5 0.5 1>, Morning Color=<rgba 1 1 1 1>, Day Color=<rgba 1 1 1 1>, Evening Color=<rgba 1 1 1 1>)
    Particles.SmallIconSize:Vector2f=&Multiply_Vector2D_ByFloat(Float=0.6, Vector2D=$Particles.SpriteSize)
   ) {isSetParameters=true}
   /SetVariables(
    Particles.DotsColor:LinearColor=&FNTimeOfDayColor(Night Color=<rgba 0.5 0.5 0.5 1>, Morning Color=<rgba 1 1 1 1>, Day Color=<rgba 1 1 1 1>, Evening Color=<rgba 1 1 1 1>)
    Particles.DotsSize:Vector2f=&Multiply_Vector2D_ByFloat(Float=0.3, Vector2D=$Particles.SpriteSize)
   ) {isSetParameters=true}
   &CurlNoiseForce(
    Bias Noise=false
    Noise Frequency=25
    Noise Strength=&RandomRangeFloat(Minimum=50, Maximum=120)
    Pan Noise=false
   )
   &Drag(Drag=&RandomRangeFloat(Minimum=0.8, Maximum=1.2), Use Linear Drag=true, Use Rotational Drag=true)
   &SolveForcesAndVelocity(Clamp Velocity=false, Limit Acceleration=false)
   &ScaleColor(
    Scale RGB=&VectorFromFloat(Value=&FNTimeOfDayScalar(Night Scale=0.5, Morning Scale=1, Evening Scale=1))
    ScaleA=true
    ScaleRGB=true
   )
  ]
  rendererClasses [&NiagaraSpriteRendererProperties]
  renderers [
   {rendererClass=&NiagaraSpriteRendererProperties}
   {rendererClass=&NiagaraSpriteRendererProperties, rendererIndex=1}
   {rendererClass=&NiagaraSpriteRendererProperties, rendererIndex=2}
  ]
 }
 MainTrail {
  bEnabled=false
  emitterUpdateScript [
   &EmitterState()
   &SpawnRate(SpawnRate=&Multiply_Float(A=$User.SpawnRate, B=300), Use Spawn Probability=false)
  ]
  particleSpawnScript [
   &InitializeParticle(Lifetime=0.35, Ribbon Width=18, Ribbon Width Mode=::NewEnumerator1, UsePositionOffset=false, Write Material Random=true)
   &SkeletalMeshLocation(Bone Sampling Mode=::NewEnumerator4, MaskSpawnGroup=false, Skeletal Mesh=$System.Skeletal Mesh Reader)
  ]
  particleUpdateScript [
   &ParticleState()
   &ScaleColor(
    Scale Alpha=&FloatFromCurve(FloatCurve=DataInterface() {propertyValues {Curve {keys [
         <at 0 0 RCIM_Cubic>
         {arriveTangent=-0.1027, interpMode=RCIM_Cubic, leaveTangent=-0.1027, tangentMode=RCTM_User, time=0.1, value=1}
         {arriveTangent=-1.1111, interpMode=RCIM_Cubic, leaveTangent=-1.1111, time=0.3762, value=0.3007}
         <at 1 0 RCIM_Cubic>
        ]}}})
    ScaleA=true
    ScaleRGB=true
   )
   &SolveForcesAndVelocity(Clamp Velocity=false, Limit Acceleration=false)
   &ScaleRibbonWidth(Ribbon Width Scale=&Multiply_Float(
     A=&FloatFromCurve(FloatCurve=DataInterface() {propertyValues {Curve {keys [
          {arriveTangent=5.2183, interpMode=RCIM_Cubic, leaveTangent=5.2183, tangentMode=RCTM_User, time=0, value=0}
          <at 0.2 1 RCIM_Cubic>
          <at 1 0 RCIM_Cubic>
         ]}}})
     B=&FloatFromCurve(
      FloatCurve=DataInterface() {propertyValues {Curve {keys [<at 0 1>, {arriveTangent=-1, interpMode=RCIM_Cubic, leaveTangent=-1, time=0.85, value=1}, <at 1 0 RCIM_Cubic>]}}}
      CurveIndex=$Emitter.NormalizedLoopAge
     )
    ))
   /SetVariables(
    Particles.1 Color:LinearColor=&MakeLinearColorFromVectorAndFloat(
     "Vector (RGB)"=<1 1 1>
     "Float (Alpha)"=&FloatFromCurve(
      FloatCurve=DataInterface() {propertyValues {Curve {keys [
          <at 0 0 RCIM_Cubic>
          {arriveTangent=-0.1027, interpMode=RCIM_Cubic, leaveTangent=-0.1027, tangentMode=RCTM_User, time=0.1, value=1}
          {arriveTangent=-1.1111, interpMode=RCIM_Cubic, leaveTangent=-1.1111, time=0.3762, value=0.3007}
          <at 1 0 RCIM_Cubic>
         ]}}}
      Scale Curve=0.475
     )
    )
    Particles.1 Width:NiagaraFloat=&Multiply_Float(A=$Particles.RibbonWidth, B=1.8)
   ) {isSetParameters=true}
   /SetVariables(
    Particles.2 Color:LinearColor=&MakeLinearColorFromVectorAndFloat(
     "Vector (RGB)"=&MakeVectorFromLinearColorRGB(Color=<rgba 0 8.3356 12 1>)
     "Float (Alpha)"=&FloatFromCurve(
      FloatCurve=DataInterface() {propertyValues {Curve {keys [
          <at 0 0 RCIM_Cubic>
          {arriveTangent=-0.1027, interpMode=RCIM_Cubic, leaveTangent=-0.1027, tangentMode=RCTM_User, time=0.1, value=1}
          {arriveTangent=-1.6667, interpMode=RCIM_Cubic, leaveTangent=-1.6667, time=0.3762, value=0.3007}
          <at 0.7 0 RCIM_Cubic>
         ]}}}
      Scale Curve=0.2
     )
    )
    Particles.2 Width:NiagaraFloat=&Multiply_Float(A=$Particles.RibbonWidth, B=0.475)
   ) {isSetParameters=true}
  ]
  rendererClasses [&NiagaraRibbonRendererProperties]
  renderers [
   {rendererClass=&NiagaraRibbonRendererProperties}
   {rendererClass=&NiagaraRibbonRendererProperties, rendererIndex=1}
   {rendererClass=&NiagaraRibbonRendererProperties, rendererIndex=2}
  ]
 }
}
summary {
 systemName=NS_Emote_DebitScent_SecondTrail
 emitters [
  {bEnabled=false, emitterName=MainTrail, rendererClasses [&NiagaraRibbonRendererProperties]}
  {bEnabled=false, emitterName=GlowTrail, rendererClasses [&NiagaraRibbonRendererProperties]}
  {emitterName=Icons, rendererClasses [&NiagaraSpriteRendererProperties]}
  {emitterName=FN_AnimTrail, rendererClasses [&NiagaraRibbonRendererProperties]}
 ]
 userVariables [
  User.Lifetime {type=NiagaraFloat, defaultValue=0, description=""}
  User.LoopDuration {type=NiagaraFloat, defaultValue=0, description=""}
  User.RibbonIDOffset {type=NiagaraInt32}
  User.SpawnRate {type=NiagaraFloat, defaultValue=1, description=""}
 ]
}
