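# Niagara system for the DebitScent emote's Showtime effect, a burst of dark and red line sprites with orange-red rays.
# BlackLines and RedLines burst, then taper spawning across three spawn groups facing separate axes, tagged for renderer visibility, with growing, fading sprites.
# Rays narrow and lengthen, their opacity also following system age.
# Time-of-day and front-end modifiers scale color; Glow has no renderers and BlackLines001 is disabled.
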NiagaraSystem /BRCosmetics/Animation/Game/MainPlayer/Emotes/DebitScent/FX/NS_Emote_DebitScent_Showtime
dependencies {usedDataInterfaces [&NiagaraDataInterfaceCurve], usedDynamicInputs [], usedModules [&SystemState]}
emitters {
 BlackLines {
  emitterUpdateScript [
   &EmitterState()
   &SpawnBurst_Instantaneous(Use Spawn Probability=false, UseLoopCountLimit=false)
   &SpawnBurst_Instantaneous(Spawn Group=1, Use Spawn Probability=false, UseLoopCountLimit=false) {module=SpawnBurst_Instantaneous001}
   &SpawnBurst_Instantaneous(Spawn Group=2, Use Spawn Probability=false, UseLoopCountLimit=false) {module=SpawnBurst_Instantaneous002}
   &SpawnRate(
    SpawnRate=&FloatFromCurve(
     FloatCurve=DataInterface() {propertyValues {Curve {keys [
         <at 0 1>
         {arriveTangent=-5.8421, interpMode=RCIM_Cubic, leaveTangent=-5.8421, tangentMode=RCTM_User, time=0.5, value=1}
         <at 0.9 0 RCIM_Cubic>
        ]}}}
     CurveIndex=$Emitter.NormalizedLoopAge
     Scale Curve=8
    )
    Use Spawn Probability=false
   )
   &SpawnRate(
    SpawnGroup=1
    SpawnRate=&FloatFromCurve(
     FloatCurve=DataInterface() {propertyValues {Curve {keys [
         <at 0 1>
         {arriveTangent=-5.8421, interpMode=RCIM_Cubic, leaveTangent=-5.8421, tangentMode=RCTM_User, time=0.5, value=1}
         <at 0.9 0 RCIM_Cubic>
        ]}}}
     CurveIndex=$Emitter.NormalizedLoopAge
     Scale Curve=8
    )
    Use Spawn Probability=false
   ) {module=SpawnRate001}
   &SpawnRate(
    SpawnGroup=2
    SpawnRate=&FloatFromCurve(
     FloatCurve=DataInterface() {propertyValues {Curve {keys [
         <at 0 1>
         {arriveTangent=-5.8421, interpMode=RCIM_Cubic, leaveTangent=-5.8421, tangentMode=RCTM_User, time=0.5, value=1}
         <at 0.9 0 RCIM_Cubic>
        ]}}}
     CurveIndex=$Emitter.NormalizedLoopAge
     Scale Curve=8
    )
    Use Spawn Probability=false
   ) {module=SpawnRate002}
  ]
  particleSpawnScript [
   /SetVariables(Particles.VisibilityTag=&SpawnGroup()) {isSetParameters=true, module=SetVariables_#1}
   &InitializeParticle(
    Color=<rgba 0.0625 0.0058 0.003 0.452>
    Lifetime Max=0.36
    Lifetime Min=0.24
    Lifetime Mode=::NewEnumerator1
    Sprite Size=&Multiply_Vector2D_ByFloat(
     Float=&Multiply_Float(A=1.25, B=&RandomRangeFloat(Minimum=1.2, Maximum=1.5))
     Vector2D=&MergeHelper_Vector2Switch(0=<250 180>, 1=<120 180>, 2=<225 120>, 3=<0 0>, 4=<0 0>, Index=&SpawnGroup(), Num=::NewEnumerator3)
    )
    Sprite Size Mode=::NewEnumerator4
    UsePositionOffset=false
    Write Material Random=true
   )
   &CameraOffset(Camera Offset Amount=-150)
   &SpriteFacingAndAlignment(
    Sprite Facing=&MergeHelper_VectorSwitch(0=<0 1 0>, 1=<1 0 0>, 2=<0 0 1>, 3=<0 0 0>, 4=<0 0 0>, Index=&SpawnGroup(), Num=::NewEnumerator3)
    WriteAlignment=false
    WriteFacing=true
   )
  ]
  particleUpdateScript [
   &ParticleState()
   &ScaleColor(
    Scale Alpha=&FloatFromCurve(
     FloatCurve=DataInterface() {propertyValues {Curve {keys [
         <at 0 0 RCIM_Cubic>
         {arriveTangent=2, interpMode=RCIM_Cubic, leaveTangent=2, time=0.4, value=1}
         {arriveTangent=-1.6667, interpMode=RCIM_Cubic, leaveTangent=-1.6667, time=0.5, value=1}
         <at 1 0 RCIM_Cubic>
        ]}}}
     Scale Curve=&FN_FrontEndFloatModifier(Gameplay=0.5, Front End=0.5)
    )
    Scale RGB=&Multiply_VectorByFloat(
     Vector=&VectorFromFloat(Value=&FNTimeOfDayScalar(Night Scale=0.5, Morning Scale=0.5, Evening Scale=0.5))
     Float=&FN_FrontEndFloatModifier(Gameplay=1, Front End=1)
    )
    ScaleA=true
    ScaleRGB=true
   )
   &ScaleSpriteSize(Uniform Curve Sprite Scale=DataInterface() {propertyValues {Curve {keys [<at 0 0 RCIM_Cubic>,<at 0.9 1 RCIM_Cubic>]}}})
   &DynamicMaterialParameters(
    Index 0 Param 1=&RandomRangeFloat(Minimum=-1, Maximum=1)
    Index 0 Param 4=&RandomRangeFloat(Minimum=0, Maximum=1)
    Module.Param0WriteEnabled=true
    Module.Param1WriteEnabled=true
    Module.Param2WriteEnabled=true
    Module.Param3WriteEnabled=true
   )
   /SetVariables(
    Particles.InnerColor:LinearColor=&MakeLinearColorFromVectorAndFloat(
     "Vector (RGB)"=&Multiply_Vector(
      A=&MakeVectorFromLinearColorRGB(Color=<rgba 0 0 0 1>)
      B=&Multiply_VectorByFloat(
       Vector=&VectorFromFloat(Value=&FNTimeOfDayScalar(Night Scale=0.5, Morning Scale=0.5, Evening Scale=0.5))
       Float=&FN_FrontEndFloatModifier(Gameplay=1, Front End=1)
      )
     )
     "Float (Alpha)"=&FloatFromCurve(
      FloatCurve=DataInterface() {propertyValues {Curve {keys [
          <at 0 0 RCIM_Cubic>
          {arriveTangent=2, interpMode=RCIM_Cubic, leaveTangent=2, time=0.4, value=1}
          {arriveTangent=-1.6667, interpMode=RCIM_Cubic, leaveTangent=-1.6667, time=0.5, value=1}
          <at 1 0 RCIM_Cubic>
         ]}}}
      Scale Curve=&Multiply_Float(A=0.325, B=&FN_FrontEndFloatModifier(Gameplay=0.5, Front End=0.5))
     )
    )
    Particles.InnerSize:Vector2f=&Multiply_Vector2D_ByFloat(Float=1.05, Vector2D=$Particles.SpriteSize)
   ) {isSetParameters=true}
  ]
  rendererClasses [&NiagaraSpriteRendererProperties]
  renderers [
   {rendererClass=&NiagaraSpriteRendererProperties}
   {rendererClass=&NiagaraSpriteRendererProperties, rendererIndex=1}
   {rendererClass=&NiagaraSpriteRendererProperties, rendererIndex=2}
   {rendererClass=&NiagaraSpriteRendererProperties, rendererIndex=3}
   {rendererClass=&NiagaraSpriteRendererProperties, rendererIndex=4}
   {rendererClass=&NiagaraSpriteRendererProperties, rendererIndex=5}
  ]
 }
 BlackLines001 {
  bEnabled=false
  emitterUpdateScript [
   &EmitterState()
   &SpawnBurst_Instantaneous(Use Spawn Probability=false, UseLoopCountLimit=false)
   &SpawnBurst_Instantaneous(Spawn Group=1, Use Spawn Probability=false, UseLoopCountLimit=false) {module=SpawnBurst_Instantaneous001}
   &SpawnBurst_Instantaneous(Spawn Group=2, Use Spawn Probability=false, UseLoopCountLimit=false) {module=SpawnBurst_Instantaneous002}
   &SpawnRate(
    SpawnRate=&FloatFromCurve(
     FloatCurve=DataInterface() {propertyValues {Curve {keys [
         <at 0 1>
         {arriveTangent=-5.8421, interpMode=RCIM_Cubic, leaveTangent=-5.8421, tangentMode=RCTM_User, time=0.5, value=1}
         <at 0.9 0 RCIM_Cubic>
        ]}}}
     CurveIndex=$Emitter.NormalizedLoopAge
     Scale Curve=8
    )
    Use Spawn Probability=false
   )
   &SpawnRate(
    SpawnGroup=1
    SpawnRate=&FloatFromCurve(
     FloatCurve=DataInterface() {propertyValues {Curve {keys [
         <at 0 1>
         {arriveTangent=-5.8421, interpMode=RCIM_Cubic, leaveTangent=-5.8421, tangentMode=RCTM_User, time=0.5, value=1}
         <at 0.9 0 RCIM_Cubic>
        ]}}}
     CurveIndex=$Emitter.NormalizedLoopAge
     Scale Curve=8
    )
    Use Spawn Probability=false
   ) {module=SpawnRate001}
   &SpawnRate(
    SpawnGroup=2
    SpawnRate=&FloatFromCurve(
     FloatCurve=DataInterface() {propertyValues {Curve {keys [
         <at 0 1>
         {arriveTangent=-5.8421, interpMode=RCIM_Cubic, leaveTangent=-5.8421, tangentMode=RCTM_User, time=0.5, value=1}
         <at 0.9 0 RCIM_Cubic>
        ]}}}
     CurveIndex=$Emitter.NormalizedLoopAge
     Scale Curve=8
    )
    Use Spawn Probability=false
   ) {module=SpawnRate002}
  ]
  particleSpawnScript [
   /SetVariables(Particles.VisibilityTag=&SpawnGroup()) {isSetParameters=true, module=SetVariables_#1}
   &InitializeParticle(
    Color=<rgba 0 0 0 0.45>
    Lifetime Max=0.36
    Lifetime Min=0.24
    Lifetime Mode=::NewEnumerator1
    Sprite Size=&Multiply_Vector2D_ByFloat(
     Float=&Multiply_Float(A=1.25, B=&RandomRangeFloat(Minimum=1.2, Maximum=1.5))
     Vector2D=&MergeHelper_Vector2Switch(0=<250 180>, 1=<120 180>, 2=<225 120>, 3=<0 0>, 4=<0 0>, Index=&SpawnGroup(), Num=::NewEnumerator3)
    )
    Sprite Size Mode=::NewEnumerator4
    UsePositionOffset=false
    Write Material Random=true
   )
   &CameraOffset(Camera Offset Amount=-150)
   &SpriteFacingAndAlignment(
    Sprite Facing=&MergeHelper_VectorSwitch(0=<0 1 0>, 1=<1 0 0>, 2=<0 0 1>, 3=<0 0 0>, 4=<0 0 0>, Index=&SpawnGroup(), Num=::NewEnumerator3)
    WriteAlignment=false
    WriteFacing=true
   )
  ]
  particleUpdateScript [
   &ParticleState()
   &ScaleColor(
    Scale Alpha=&FloatFromCurve(
     FloatCurve=DataInterface() {propertyValues {Curve {keys [
         <at 0 0 RCIM_Cubic>
         {arriveTangent=2, interpMode=RCIM_Cubic, leaveTangent=2, time=0.4, value=1}
         {arriveTangent=-1.6667, interpMode=RCIM_Cubic, leaveTangent=-1.6667, time=0.5, value=1}
         <at 1 0 RCIM_Cubic>
        ]}}}
     Scale Curve=&FN_FrontEndFloatModifier(Gameplay=1, Front End=0.5)
    )
    Scale RGB=&Multiply_VectorByFloat(
     Vector=&VectorFromFloat(Value=&FNTimeOfDayScalar(Night Scale=0.5, Morning Scale=0.25, Evening Scale=0.5))
     Float=&FN_FrontEndFloatModifier(Gameplay=1, Front End=1)
    )
    ScaleA=true
    ScaleRGB=true
   )
   &ScaleSpriteSize(Uniform Curve Sprite Scale=DataInterface() {propertyValues {Curve {keys [<at 0 0 RCIM_Cubic>,<at 0.9 1 RCIM_Cubic>]}}})
   &DynamicMaterialParameters(
    Index 0 Param 1=&RandomRangeFloat(Minimum=-1, Maximum=1)
    Index 0 Param 4=&RandomRangeFloat(Minimum=0, Maximum=1)
    Module.Param0WriteEnabled=true
    Module.Param1WriteEnabled=true
    Module.Param2WriteEnabled=true
    Module.Param3WriteEnabled=true
   )
  ]
  rendererClasses [&NiagaraSpriteRendererProperties]
  renderers [
   {rendererClass=&NiagaraSpriteRendererProperties}
   {rendererClass=&NiagaraSpriteRendererProperties, rendererIndex=1}
   {rendererClass=&NiagaraSpriteRendererProperties, rendererIndex=2}
  ]
 }
 Glow {rendererClasses [], renderers []}
 Rays {
  emitterUpdateScript [
   &EmitterState()
   &SpawnRate(
    SpawnRate=&FloatFromCurve(
     FloatCurve=DataInterface() {propertyValues {Curve {keys [
         <at 0 1>
         {arriveTangent=-5.8421, interpMode=RCIM_Cubic, leaveTangent=-5.8421, tangentMode=RCTM_User, time=0.5, value=1}
         <at 0.9 0 RCIM_Cubic>
        ]}}}
     CurveIndex=$Emitter.NormalizedLoopAge
     Scale Curve=30
    )
    Use Spawn Probability=false
   )
   &SpawnBurst_Instantaneous(Spawn Count=3, Use Spawn Probability=false, UseLoopCountLimit=false)
  ]
  particleSpawnScript [
   &InitializeParticle(
    Color=<rgba 1 0.071 0 0.625>
    Lifetime Max=0.35
    Lifetime Min=0.15
    Lifetime Mode=::NewEnumerator1
    Sprite Size=&Multiply_Vector2D_ByFloat(Float=1.25, Vector2D=&RandomRangeVector2D(Minimum=<10 120>, Maximum=<80 140>))
    Sprite Size Mode=::NewEnumerator4
    UsePositionOffset=false
    Write Material Random=true
   )
   &CameraOffset()
   &ShapeLocation(MaskSpawnGroup=false, Non Uniform Scale=<3 1 1>, Override Seed=false, Sphere Radius=12, UseOwnerScale=true)
   &AddVelocity(Module.Use Falloff Distance=false, Override Seed=false, Velocity Mode=::NewEnumerator1, Velocity Speed=1)
   &SubUVAnimation(End Frame=1, SubUV Animation Mode:ENiagara_SubUVLookupMode=::NewEnumerator2)
  ]
  particleUpdateScript [
   &ParticleState()
   &ScaleSpriteSize(
    Scale Factor=&MakeVector2D(
     X=&FloatFromCurve(FloatCurve=DataInterface() {propertyValues {Curve {keys [<at 0 1>,<at 1 0>]}}})
     Y=&FloatFromCurve(FloatCurve=DataInterface() {propertyValues {Curve {keys [<at 0 0.5>,<at 1 1>]}}})
    )
    Scale Sprite Size Mode=::NewEnumerator2
   )
   &ScaleColor(
    Scale Alpha=&Multiply_Float(
     A=&FloatFromCurve(
      FloatCurve=DataInterface() {propertyValues {Curve {keys [{interpMode=RCIM_Cubic, tangentMode=RCTM_User, time=0, value=0}, {interpMode=RCIM_Cubic, tangentMode=RCTM_User, time=0.4, value=1}, <at 1 0>]}}}
      Scale Curve=0.5
     )
     B=&FloatFromCurve(
      FloatCurve=DataInterface() {propertyValues {Curve {keys [
          <at 0 0.5>
          {arriveTangent=0.1737, interpMode=RCIM_Cubic, leaveTangent=0.1737, tangentMode=RCTM_User, time=0.2, value=1.35}
          {interpMode=RCIM_Cubic, tangentMode=RCTM_User, time=0.6, value=1.3}
          {arriveTangent=-0.5397, interpMode=RCIM_Cubic, leaveTangent=-0.5397, tangentMode=RCTM_User, time=0.8, value=0.6}
          {arriveTangent=-0.405, interpMode=RCIM_Cubic, leaveTangent=-0.405, tangentMode=RCTM_User, time=1.3, value=0}
         ]}}}
      CurveIndex=$Engine.System.Age
     )
    )
    Scale RGB=&Multiply_VectorByFloat(
     Vector=&VectorFromFloat(Value=&FNTimeOfDayScalar(Night Scale=0.5, Morning Scale=0.25, Evening Scale=0.5))
     Float=&FN_FrontEndFloatModifier(Gameplay=1, Front End=1)
    )
    ScaleA=true
    ScaleRGB=true
   )
   &SolveForcesAndVelocity(Clamp Velocity=false, Limit Acceleration=false)
  ]
  rendererClasses [&NiagaraSpriteRendererProperties]
  renderers [{rendererClass=&NiagaraSpriteRendererProperties}]
 }
 RedLines {
  emitterUpdateScript [
   &EmitterState()
   &SpawnBurst_Instantaneous(Use Spawn Probability=false, UseLoopCountLimit=false)
   &SpawnBurst_Instantaneous(Spawn Group=1, Use Spawn Probability=false, UseLoopCountLimit=false) {module=SpawnBurst_Instantaneous001}
   &SpawnBurst_Instantaneous(Spawn Group=2, Use Spawn Probability=false, UseLoopCountLimit=false) {module=SpawnBurst_Instantaneous002}
   &SpawnRate(
    SpawnRate=&FloatFromCurve(
     FloatCurve=DataInterface() {propertyValues {Curve {keys [
         <at 0 1>
         {arriveTangent=-5.8421, interpMode=RCIM_Cubic, leaveTangent=-5.8421, tangentMode=RCTM_User, time=0.5, value=1}
         <at 0.9 0 RCIM_Cubic>
        ]}}}
     CurveIndex=$Emitter.NormalizedLoopAge
     Scale Curve=8
    )
    Use Spawn Probability=false
   )
   &SpawnRate(
    SpawnGroup=1
    SpawnRate=&FloatFromCurve(
     FloatCurve=DataInterface() {propertyValues {Curve {keys [
         <at 0 1>
         {arriveTangent=-5.8421, interpMode=RCIM_Cubic, leaveTangent=-5.8421, tangentMode=RCTM_User, time=0.5, value=1}
         <at 0.9 0 RCIM_Cubic>
        ]}}}
     CurveIndex=$Emitter.NormalizedLoopAge
     Scale Curve=8
    )
    Use Spawn Probability=false
   ) {module=SpawnRate001}
   &SpawnRate(
    SpawnGroup=2
    SpawnRate=&FloatFromCurve(
     FloatCurve=DataInterface() {propertyValues {Curve {keys [
         <at 0 1>
         {arriveTangent=-5.8421, interpMode=RCIM_Cubic, leaveTangent=-5.8421, tangentMode=RCTM_User, time=0.5, value=1}
         <at 0.9 0 RCIM_Cubic>
        ]}}}
     CurveIndex=$Emitter.NormalizedLoopAge
     Scale Curve=8
    )
    Use Spawn Probability=false
   ) {module=SpawnRate002}
  ]
  particleSpawnScript [
   /SetVariables(Particles.VisibilityTag=&SpawnGroup()) {isSetParameters=true, module=SetVariables_#1}
   &InitializeParticle(
    Color=<rgba 0.5647 0.1356 0.1356 0.8>
    Lifetime Max=0.36
    Lifetime Min=0.24
    Lifetime Mode=::NewEnumerator1
    Sprite Size=&Multiply_Vector2D_ByFloat(
     Float=&Multiply_Float(A=0.8, B=&RandomRangeFloat(Minimum=1.2, Maximum=1.5))
     Vector2D=&MergeHelper_Vector2Switch(0=<250 180>, 1=<120 180>, 2=<225 120>, 3=<0 0>, 4=<0 0>, Index=&SpawnGroup(), Num=::NewEnumerator3)
    )
    Sprite Size Mode=::NewEnumerator4
    UsePositionOffset=false
    Write Material Random=true
   )
   &CameraOffset(Camera Offset Amount=-10)
   &SpriteFacingAndAlignment(
    Sprite Facing=&MergeHelper_VectorSwitch(0=<0 1 0>, 1=<1 0 0>, 2=<0 0 1>, 3=<0 0 0>, 4=<0 0 0>, Index=&SpawnGroup(), Num=::NewEnumerator3)
    WriteAlignment=false
    WriteFacing=true
   )
  ]
  particleUpdateScript [
   &ParticleState()
   &ScaleColor(
    Scale Alpha=&FloatFromCurve(
     FloatCurve=DataInterface() {propertyValues {Curve {keys [
         <at 0 0 RCIM_Cubic>
         {arriveTangent=2, interpMode=RCIM_Cubic, leaveTangent=2, time=0.4, value=1}
         {arriveTangent=-1.6667, interpMode=RCIM_Cubic, leaveTangent=-1.6667, time=0.5, value=1}
         <at 1 0 RCIM_Cubic>
        ]}}}
     Scale Curve=&FN_FrontEndFloatModifier(Gameplay=0.5, Front End=0.5)
    )
    Scale RGB=&Multiply_VectorByFloat(
     Vector=&VectorFromFloat(Value=&FNTimeOfDayScalar(Night Scale=0.5, Morning Scale=0.25, Evening Scale=0.5))
     Float=&FN_FrontEndFloatModifier(Gameplay=1, Front End=1)
    )
    ScaleA=true
    ScaleRGB=true
   )
   &ScaleSpriteSize(Uniform Curve Sprite Scale=DataInterface() {propertyValues {Curve {keys [<at 0 0 RCIM_Cubic>,<at 0.9 1 RCIM_Cubic>]}}})
   &DynamicMaterialParameters(
    Index 0 Param 1=&RandomRangeFloat(Minimum=-1, Maximum=1)
    Index 0 Param 4=&RandomRangeFloat(Minimum=0, Maximum=1)
    Module.Param0WriteEnabled=true
    Module.Param1WriteEnabled=true
    Module.Param2WriteEnabled=true
    Module.Param3WriteEnabled=true
   )
  ]
  rendererClasses [&NiagaraSpriteRendererProperties]
  renderers [
   {rendererClass=&NiagaraSpriteRendererProperties}
   {rendererClass=&NiagaraSpriteRendererProperties, rendererIndex=1}
   {rendererClass=&NiagaraSpriteRendererProperties, rendererIndex=2}
  ]
 }
}
summary {
 systemName=NS_Emote_DebitScent_Showtime
 emitters [
  {emitterName=BlackLines, rendererClasses [&NiagaraSpriteRendererProperties]}
  {emitterName=Rays, rendererClasses [&NiagaraSpriteRendererProperties]}
  {emitterName=RedLines, rendererClasses [&NiagaraSpriteRendererProperties]}
  {emitterName=Glow, rendererClasses []}
  {bEnabled=false, emitterName=BlackLines001, rendererClasses [&NiagaraSpriteRendererProperties]}
 ]
 userVariables [
  User.bIsFrontEnd {type=NiagaraBool}
  User.bIsFrontEndPreview {type=NiagaraBool}
 ]
}
